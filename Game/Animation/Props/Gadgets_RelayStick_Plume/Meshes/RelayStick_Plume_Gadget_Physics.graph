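# Physics asset for the RelayStick Plume gadget prop, providing simplified collision and joint setup for its skeletal mesh.
# Capsule bodies cover the blade bone and five Pen bones, JNT_Pen_02 through JNT_Pen_06.
# Constraints link the Pen bodies into a serial chain, each joint sharing the same moderate swing and twist limits.
# The blade body has no constraint, and none joins the chain to it.
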PhysicsAsset /Game/Animation/Props/Gadgets_RelayStick_Plume/Meshes/RelayStick_Plume_Gadget_Physics
bodies {
 JNT_Blade [{shapeType=Capsule, length=34.2607, radius=7.911, center=<-27.0326 -0.9523 -3.1127>, rotation=<rot 0 90.2103 -85.4717>}]
 JNT_Pen_02 [{shapeType=Capsule, length=17.3169, radius=12.7974, center=<2.5424 -15.4033 0.9231>, rotation=<rot 0 81.8072 -88.8667>}]
 JNT_Pen_03 [{shapeType=Capsule, length=21.6718, radius=17.4469, center=<-0.8103 -7.1367 2.2233>, rotation=<rot 0 -81.2806 -89.6862>}]
 JNT_Pen_04 [{shapeType=Capsule, length=23.5896, radius=16.3245, center=<-1.0436 -11.6629 3.6868>, rotation=<rot 0 -80.1546 -85.8733>}]
 JNT_Pen_05 [{shapeType=Capsule, length=23.3304, radius=13.3562, center=<-2.4359 -8.306 5.5106>, rotation=<rot 0 -74.8999 -83.0498>}]
 JNT_Pen_06 [{shapeType=Capsule, length=25.0169, radius=18.5713, center=<-2.863 -19.6637 1.6598>, rotation=<rot 0 170.4675 -82.8239>}]
}
constraints [
 {bone1Name=JNT_Pen_03, bone2Name=JNT_Pen_02, swing1LimitDegrees=45, swing2LimitDegrees=45, twistLimitDegrees=45}
 {bone1Name=JNT_Pen_04, bone2Name=JNT_Pen_03, swing1LimitDegrees=45, swing2LimitDegrees=45, twistLimitDegrees=45}
 {bone1Name=JNT_Pen_05, bone2Name=JNT_Pen_04, swing1LimitDegrees=45, swing2LimitDegrees=45, twistLimitDegrees=45}
 {bone1Name=JNT_Pen_06, bone2Name=JNT_Pen_05, swing1LimitDegrees=45, swing2LimitDegrees=45, twistLimitDegrees=45}
]
